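# Translucent lit Niagara sprite material for mist and water droplets, named for use at a waterfall base.
# A slowly panning smoke texture, distorted by normal-map noise, forms mist faded by UV gradients, a sphere mask and particle alpha.
# Faster-panning dirt-bit and foam textures sum into droplets, eroded along U by a smoothstep and tinted by particle color.
# Dynamic parameters scale horizontal tiling (Width) and droplet intensity (DropletsStr).
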Material /Game/Effects/Fort_Effects/Effects/Water/Materials/M_FX_WaterfallBase_MistAndWater
graph {
 properties {bUsedWithNiagaraSprites=true, blendMode=BLEND_Translucent, pixelDepthOffsetMode=PDOM_Legacy, usedShadingModels=Default Lit}
 nodes [
  Constant3Vector()
  Constant()
  n1 = ParticleColor()
  n2 = TextureCoordinate() {uTiling=0.5, vTiling=2}
  n3 = Multiply(A=ComponentMask(None=TextureSample(UVs=@n2) {samplerType=SAMPLERTYPE_Normal, texture=/Game/Effects/Fort_Effects/Textures/Tile/T_GaussianNoise_02_N}.RGB) {g=true, r=true}) {constB=0.05}
  n4 = TextureCoordinate() {uTiling=0.5}
  n5 = DynamicParameter() {paramNames [Width,DropletsStr,Param3,Param4]}
  n6 = AppendVector(A=Constant() {r=1}, B=Divide(A=@n5.Width) {constB=350})
  n7 = Add(A=@n3, B=Multiply(A=Multiply(A=@n4, B=@n6)) {constB=0.5})
  n8 = TextureCoordinate()
  n9 = Multiply(
   A=TextureSample(UVs=Panner(Coordinate=@n7) {bFractionalPart=true, speedX=-0.15}) {texture=/Game/Global/Textures/Default/Tileable_Patterns/T_Smoke_Puff_Tile_Up}.G
   B=OneMinus(None=ComponentMask(None=@n8) {r=true})
  )
  n10 = SphereMask(A=ComponentMask(None=TextureCoordinate()) {g=true}, B=Constant() {r=0.5}) {attenuationRadius=0.5, hardnessPercent=80}
  n11 = TextureCoordinate() {uTiling=4}
  Multiply(A=Multiply(A=@n1.A, B=Multiply(A=@n9, B=@n10)), B=Saturate(None=ComponentMask(None=@n11) {r=true}))
  n12 = Add(A=@n3, B=Multiply(A=@n6, B=TextureCoordinate() {uTiling=0.5, vTiling=1.5}))
  n13 = Add(
   A=TextureSample(UVs=Panner(Coordinate=@n12) {bFractionalPart=true, speedX=1.7, speedY=-0.1}) {samplerType=SAMPLERTYPE_LinearGrayscale, texture=/Game/Effects/Fort_Effects/Textures/Rocks/T_Tiling_Dust_Dirt_Bits_M}.R
   B=TextureSample(UVs=Panner(Coordinate=@n12) {bFractionalPart=true, speedX=1, speedY=0.1}) {samplerType=SAMPLERTYPE_LinearColor, texture=/Game/Global/Textures/Default/Caustics/WaterFoam_AccentSwirls_H}.R
  )
  n14 = TextureCoordinate()
  Multiply(A=@n13, B=OneMinus(None=ComponentMask(None=@n14) {r=true}))
  n15 = ComponentMask(None=TextureCoordinate()) {r=true}
  n16 = ScalarParameter() {parameterName=sste}
  Multiply(A=SmoothStep(Min=@n15, Max=Add(A=@n15, B=@n16), Value=@n13), B=Multiply(A=@n1.RGB, B=@n5.DropletsStr)) {constB=2}
 ]
}
