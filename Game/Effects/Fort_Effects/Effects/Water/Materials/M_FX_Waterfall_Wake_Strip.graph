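# Unlit translucent Niagara sprite material that, going by its name, draws an animated foam wake strip for waterfalls.
# UVs are warped by a square root along U, tiled by a dynamic Width parameter and distorted with panning Gaussian noise normals.
# They scroll two caustics samples and a foam swirl, which are averaged, faded toward strip edges by sphere masks and scaled by particle alpha.
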Material /Game/Effects/Fort_Effects/Effects/Water/Materials/M_FX_Waterfall_Wake_Strip
graph {
 properties {bUsedWithNiagaraSprites=true, blendMode=BLEND_Translucent, pixelDepthOffsetMode=PDOM_Legacy, shadingModel=MSM_Unlit, usedShadingModels=Unlit}
 nodes [
  n1 = Constant2Vector() {g=1, r=1.5}
  n2 = TextureCoordinate() {uTiling=0.5, vTiling=0.5}
  n3 = ComponentMask(None=@n2) {r=true}
  n4 = TextureSample(UVs=TextureCoordinate()) {samplerType=SAMPLERTYPE_Normal, texture=/Game/Effects/Fort_Effects/Textures/Tile/T_GaussianNoise_02_N}
  n5 = Add(A=@n2, B=Multiply(A=ComponentMask(None=@n4.RGB) {g=true, r=true}) {constB=0.05})
  n6 = DynamicParameter() {paramNames [Width,Param2,Param3,Param4]}
  n7 = Multiply(
   A=AppendVector(A=Power(Base=@n3) {constExponent=0.5}, B=ComponentMask(None=@n5) {g=true})
   B=AppendVector(A=Constant() {r=2}, B=Divide(A=@n6.Width) {constB=200})
  )
  n8 = TextureSample(UVs=Panner(Coordinate=Multiply(A=@n1, B=@n7)) {speedX=0.5}) {samplerType=SAMPLERTYPE_Normal, texture=&T_GaussianNoise_02_N}
  n9 = Add(A=Multiply(A=ComponentMask(None=@n8.RGB) {g=true, r=true}) {constB=0.125}, B=@n7)
  n10 = Panner(Coordinate=@n9) {speedX=0.6, speedY=-0.1}
  n11 = Panner(Coordinate=@n9) {speedX=0.4, speedY=0.1}
  n12 = Add(
   A=Add(
    A=TextureSample(UVs=@n10) {texture=/Game/Effects/Fort_Effects/Textures/Tile/T_ShieldCaustics_Tile_Chromatic}.R
    B=TextureSample(UVs=@n11) {texture=&T_ShieldCaustics_Tile_Chromatic}.G
   )
   B=TextureSample(UVs=Multiply(A=@n10) {constB=2}) {samplerType=SAMPLERTYPE_LinearColor, texture=/Game/Global/Textures/Default/Caustics/WaterFoam_AccentSwirls_H}.G
  )
  n13 = SphereMask(A=ComponentMask(None=TextureCoordinate()) {r=true}, B=Constant() {r=0.5}) {attenuationRadius=0.5, hardnessPercent=0}
  n14 = SphereMask(A=ComponentMask(None=TextureCoordinate()) {g=true}, B=Constant() {r=0.5}) {attenuationRadius=0.5, hardnessPercent=60}
  Multiply(A=ParticleColor().A, B=Multiply(A=Divide(A=@n12) {constB=3}, B=Multiply(A=@n13, B=@n14)))
 ]
}
