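# Niagara system for a subtle waterfall base effect, spreading foam, droplets and mist along a line sized by User.Source_Length.
# Two toggleable single-sprite layers draw the base and back mist, passing length to their materials.
# GPU droplet and mist emitters spawn along that line, launch upward, fall under gravity and die in kill volumes; only droplet rate scales with length.
# Fountain, splash, ripple and extra mesh emitters are disabled.
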NiagaraSystem /Game/Effects/Fort_Effects/Effects/Water/NS_SubtleWaterfallBase_Subtle
dependencies {
 usedDataInterfaces [
  &NiagaraDataInterfaceCurve
  &NiagaraDataInterfaceColorCurve
  &NiagaraDataInterfaceVector2DCurve
  &NiagaraDataInterfaceVectorCurve
 ]
 usedDynamicInputs []
 usedModules [&SystemState]
}
emitters {
 BaseMesh {
  emitterUpdateScript [
   &EmitterState(Life Cycle Mode=::NewEnumerator1, Loop Behavior=::NewEnumerator1, Loop Duration Mode=::NewEnumerator1, UseLoopDelay=false)
   &SpawnBurst_Instantaneous(
    Spawn Count=&MakeCustomIntFromBool(Bool=$User.bUseSourceCaustics, True INT=1, False INT=0)
    Use Spawn Probability=false
    UseLoopCountLimit=false
   )
  ]
  particleSpawnScript [&InitializeParticle(
    Color=$User.BaseColor
    Lifetime=$Emitter.CurrentLoopDuration
    Sprite Size=&MakeVector2D(X=$User.BaseWidth, Y=$User.Source_Length)
    Sprite Size Mode=::NewEnumerator4
    UsePositionOffset=true
    Write Material Random=true
   )]
  particleUpdateScript [
   &ParticleState(
    Kill Particles When Lifetime Has Elapsed=false
    Let Infinitely Lived Particles Die When Emitter Deactivates=true
    Loop Particles Lifetime=true
   )
   &SolveForcesAndVelocity(Clamp Velocity=false, Limit Acceleration=false)
   &SpriteFacingAndAlignment(Facing Coordinate Space=::World, Sprite Facing=<0 0 1>, WriteAlignment=true, WriteFacing=true)
   &DynamicMaterialParameters(
    Index 0 Param 1=$User.Source_Length
    Module.Index0Param0Write=true
    Module.Index0Param1Write=true
    Module.Index0Param2Write=true
    Module.Index0Param3Write=true
   )
  ]
  rendererClasses [&NiagaraSpriteRendererProperties]
  renderers [{rendererClass=&NiagaraSpriteRendererProperties}]
 }
 FN_Mesh001 {
  bEnabled=false
  emitterUpdateScript [
   &EmitterState()
   &SpawnRate(SpawnRate=10, Use Spawn Probability=false)
  ]
  particleSpawnScript [
   /Niagara/Modules/Spawn/Initialization/InitializeParticle(
    Color=<rgba 0.5 0.5 0.5 0.002>
    Lifetime=&UniformRangedFloat(Minimum=1.9, Maximum=2.4)
    Mesh Scale=<3 3 1>
    Module.Write Color=true
    Module.Write Lifetime=true
    Module.Write Mass=true
    Module.Write Material Random=true
    Module.Write Position=true
    Module.Write Scale=true
    Module.Write SpriteRotation=false
    Module.Write SpriteSize=false
    Position=&Add_Vector(A=/Niagara/DynamicInputs/Helpers/SimulationPosition(), B=<0 0 -10>)
   )
   /Niagara/Modules/Spawn/Orientation/V2/InitialMeshOrientation(
    Orientation Axis:Vector3f=<0 0 1>
    Orientation Coordinate Space:ENiagaraCoordinateSpace=::World
    Orientation Vector=<0 0 1>
    Rotation=&RandomRangeVector(Minimum=<0 0 360>, Maximum=<0 0 0>)
    Use Orientation Vector=true
    Use Rotation Vector=true
   ) {enabled=false}
   /Niagara/Modules/Spawn/Orientation/V3/InitialMeshOrientation(
    Mesh Orientation Mode=::NewEnumerator1
    Rotation=&RandomRangeVector(Minimum=<0 0 360>, Maximum=<0 0 0>)
    Use Rotation Vector=true
   ) {module=InitialMeshOrientation001}
   &ShapeLocation(
    Box Size=&MakeVector(X=1, Y=$User.Source_Length, Z=1)
    MaskSpawnGroup=false
    Offset=<0 0 -20>
    Offset Mode=::NewEnumerator0
    Override Seed=false
    Shape Primitive=::NewEnumerator2
    Surface Only=false
    UseOwnerScale=true
   )
  ]
  particleUpdateScript [
   &ParticleState()
   &ScaleMeshSize(Scale Factor=&VectorFromCurve(VectorCurve=DataInterface() {propertyValues {
       XCurve {keys [
         {interpMode=RCIM_Cubic, leaveTangent=7.6613, tangentMode=RCTM_Break, time=0, value=0}
         {arriveTangent=2.4936, interpMode=RCIM_Cubic, leaveTangent=2.4936, tangentMode=RCTM_User, time=0.2392, value=1.3545}
         <at 1 2>
        ]}
       YCurve {keys [
         {interpMode=RCIM_Cubic, leaveTangent=7.7916, tangentMode=RCTM_Break, time=0, value=0}
         {arriveTangent=2.6017, interpMode=RCIM_Cubic, leaveTangent=2.6017, tangentMode=RCTM_User, time=0.2392, value=1.3545}
         <at 1 2>
        ]}
       ZCurve {keys [<at 0 3>,<at 1 1>]}
      }}))
   &DynamicMaterialParameters(
    Index 0 Param 1=&FloatFromCurve(FloatCurve=DataInterface() {propertyValues {Curve {keys [<at 0 0>,<at 1 1.5>]}}})
    Index 0 Param 2=0.25
    Index 0 Param 3=50
    Index 0 Param 4=1
    Module.Param0WriteEnabled=true
    Module.Param1WriteEnabled=true
    Module.Param2WriteEnabled=true
    Module.Param3WriteEnabled=true
   )
   &CameraOffset(Camera Offset Amount=50)
   &ScaleColor(
    Scale Alpha=&FloatFromCurve(FloatCurve=DataInterface() {propertyValues {Curve {keys [
         <at 0 0>
         {arriveTangent=5.1071, interpMode=RCIM_Cubic, leaveTangent=5.1071, tangentMode=RCTM_Break, time=0.1294, value=2.8567}
         {arriveTangent=-8.2025, interpMode=RCIM_Cubic, leaveTangent=-8.2025, tangentMode=RCTM_User, time=0.3115, value=1.8534}
         <at 0.4715 0.0687>
         <at 1 0>
        ]}}})
    Scale RGB=&VectorFromFloat(Value=&FloatFromCurve(FloatCurve=DataInterface() {propertyValues {Curve {keys [<at 0 2>,<at 1 0.2>]}}}))
    ScaleA=true
    ScaleRGB=true
   )
  ]
  rendererClasses [&NiagaraMeshRendererProperties]
  renderers [{rendererClass=&NiagaraMeshRendererProperties, rendererIndex=0}]
 }
 FN_Water_Drops_ {
  simTarget=GPUComputeSim
  emitterUpdateScript [
   &EmitterState(Enable Distance Culling=true, MaxDistance=3500, Module.bUseMaxDistance=true, Module.bUseMinDistance=false, Scalability Mode=::NewEnumerator1)
   &SpawnRate(
    SpawnRate=&Multiply_Float(A=&Divide_Float(A=$User.Source_Length, B=10), B=$User.DropletRateMultiplier)
    Use Spawn Probability=true
   ) {module=SpawnRate001}
  ]
  particleSpawnScript [
   &InitializeParticle(
    Lifetime=&UniformRangedFloat(Minimum=0.5, Maximum=1.5)
    Mass=&UniformRangedFloat(Minimum=0.6, Maximum=1.2)
    Mesh Scale=<100 100 100>
    Module.Write Color=true
    Module.Write Lifetime=true
    Module.Write Mass=true
    Module.Write Material Random=true
    Module.Write Position=false
    Module.Write Scale=false
    Module.Write SpriteRotation=true
    Module.Write SpriteSize=true
    Position=/Niagara/Functions/Localspace/SimulationPosition()
    Sprite Size=&Vector2DFromCurve(
     Vector2Curve=DataInterface() {propertyValues {
       XCurve {keys [<at -0.003 3.8449 RCIM_Cubic>,<at 0.9989 0.3489 RCIM_Cubic>]}
       YCurve {keys [
         {arriveTangent=-30.8979, interpMode=RCIM_Cubic, leaveTangent=-30.8979, tangentMode=RCTM_User, time=0.001, value=13.7761}
         {arriveTangent=-11.8685, interpMode=RCIM_Cubic, leaveTangent=-11.8685, time=0.2536, value=8.9484}
         <at 0.9989 1.9328 RCIM_Cubic>
        ]}
      }}
     CurveIndex=$Particles.Random
     ScaleCurve=&RandomRangeVector2D(Minimum=<0.8 0.8>, Maximum=<1.2 1.2>)
    )
   ) {enabled=false}
   /Niagara/Modules/Spawn/Initialization/V2/InitializeParticle(
    Color=<rgba 0.2784 0.3908 0.5 1>
    Lifetime=&UniformRangedFloat(Minimum=0.5, Maximum=1.5)
    Mass=&UniformRangedFloat(Minimum=0.6, Maximum=1.2)
    Mass Mode=::NewEnumerator1
    Sprite Rotation Mode=::NewEnumerator0
    Sprite Size=&Vector2DFromCurve(
     Vector2Curve=DataInterface() {propertyValues {
       XCurve {keys [<at -0.003 3.8449 RCIM_Cubic>,<at 0.9989 0.3489 RCIM_Cubic>]}
       YCurve {keys [{arriveTangent=-11.8685, interpMode=RCIM_Cubic, time=0.3, value=5}, <at 0.9989 1.9328 RCIM_Cubic>]}
      }}
     ScaleCurve=&RandomRangeVector2D(Minimum=<0.8 0.8>, Maximum=<1.2 1.2>)
    )
    Sprite Size Mode=::NewEnumerator4
    UsePositionOffset=false
    Write Material Random=true
   ) {module=InitializeParticle001}
   &BoxLocation(Box Size=$User.BoxSize, MaskSpawnGroup=false, Module.Override Seed=false, Rotate Box=false, Surface Only=false) {enabled=false}
   &ShapeLocation(
    Box Size=&MakeVector(X=1, Y=$User.Source_Length, Z=1)
    MaskSpawnGroup=false
    Offset=<0 0 -20>
    Offset Mode=::NewEnumerator0
    Override Seed=false
    Shape Primitive=::NewEnumerator2
    Surface Only=false
    UseOwnerScale=true
   )
   &AddVelocity(
    Cone Angle=50
    Cone Axis=<0.5 0 1>
    Module.UseConeFalloff=false
    Module.UseVelDistribution=false
    Override Seed=false
    Velocity Mode=::NewEnumerator2
    Velocity Speed=&RandomRangeFloat(Minimum=250, Maximum=500)
   )
   /SetVariables(Particles.Random=&UniformRangedFloat(Minimum=0, Maximum=1)) {isSetParameters=true}
   &DynamicMaterialParameters(
    Index 0 Param 1=&UniformRangedFloat(Minimum=4, Maximum=10)
    Module.Param0WriteEnabled=true
    Module.Param1WriteEnabled=false
    Module.Param2WriteEnabled=false
    Module.Param3WriteEnabled=false
   )
  ]
  particleUpdateScript [
   &ParticleState()
   &ScaleColor(
    Scale Alpha=&FloatFromCurve(
     FloatCurve=DataInterface() {propertyValues {Curve {keys [
         <at 0 0 RCIM_Cubic>
         {arriveTangent=0.002, interpMode=RCIM_Cubic, leaveTangent=0.002, time=0.4, value=1.05}
         {arriveTangent=-4.2194, interpMode=RCIM_Cubic, leaveTangent=-4.2194, tangentMode=RCTM_User, time=1.0011, value=0.002}
        ]}}}
     Scale Curve=0.15
    )
    ScaleA=true
    ScaleRGB=true
   )
   &Drag(Drag=1.2, Use Linear Drag=true, Use Rotational Drag=false)
   &GravityForce()
   &SolveForcesAndVelocity(Clamp Velocity=false, Limit Acceleration=false)
   &ScaleSpriteSize(Uniform Curve Sprite Scale=DataInterface() {propertyValues {Curve {keys [
        {interpMode=RCIM_Cubic, tangentMode=RCTM_User, time=0.5, value=1}
        {arriveTangent=-4.4169, interpMode=RCIM_Cubic, leaveTangent=-4.4169, tangentMode=RCTM_User, time=1, value=0}
       ]}}})
   &KillParticlesInVolume(Kill Shape=::NewEnumerator2, Origin Offset:Vector3f=<0 0 -50>)
  ]
  rendererClasses [&NiagaraSpriteRendererProperties]
  renderers [{rendererClass=&NiagaraSpriteRendererProperties}]
 }
 Flat_Ripple {
  bEnabled=false
  emitterUpdateScript [
   &EmitterState()
   &SpawnRate(
    SpawnRate=&Multiply_Float(A=&Divide_Float(A=$User.Source_Length, B=450), B=$User.RipplesSpawnMultiplier)
    Use Spawn Probability=false
   )
   &SpawnBurst_Instantaneous(Use Spawn Probability=false, UseLoopCountLimit=false) {enabled=false}
  ]
  particleSpawnScript [
   &InitializeParticle(
    Color=$User.RipplesColor
    Lifetime=&RandomRangeFloat(Minimum=1.5, Maximum=3)
    Mesh Scale=&Multiply_VectorByFloat(Vector=<8 8 1>, Float=0.2)
    Mesh Scale Mode=::NewEnumerator4
    UsePositionOffset=false
    Write Material Random=true
   ) {module=InitializeParticle001}
   &InitialMeshOrientation(
    Mesh Orientation Mode=::NewEnumerator7
    Rotation=&RandomRangeVector(Minimum=<0 0 0>, Maximum=<0 0 1>)
    Use Rotation Vector=true
   )
   &ShapeLocation(
    Box Size=&MakeVector(X=1, Y=$User.Source_Length, Z=1)
    MaskSpawnGroup=false
    Offset=<0 0 -2>
    Offset Mode=::NewEnumerator0
    Override Seed=false
    Shape Primitive=::NewEnumerator2
    Surface Only=false
    UseOwnerScale=true
   )
  ]
  particleUpdateScript [
   &ParticleState()
   &ScaleMeshSize(Scale Factor=&VectorFromCurve(VectorCurve=DataInterface() {propertyValues {
       XCurve {keys [
         {interpMode=RCIM_Cubic, leaveTangent=7.6613, tangentMode=RCTM_Break, time=0, value=0}
         {arriveTangent=2.4936, interpMode=RCIM_Cubic, leaveTangent=2.4936, tangentMode=RCTM_User, time=0.2392, value=1.3545}
         <at 1 2>
        ]}
       YCurve {keys [
         {interpMode=RCIM_Cubic, leaveTangent=7.7916, tangentMode=RCTM_Break, time=0, value=0}
         {arriveTangent=2.6017, interpMode=RCIM_Cubic, leaveTangent=2.6017, tangentMode=RCTM_User, time=0.2392, value=1.3545}
         <at 1 2>
        ]}
       ZCurve {keys [<at 0 3>,<at 1 1>]}
      }}))
   &DynamicMaterialParameters(
    Index 0 Param 1=&FloatFromCurve(FloatCurve=DataInterface() {propertyValues {Curve {keys [<at 0 0>,<at 1 1.5>]}}})
    Index 0 Param 2=&FloatFromCurve(FloatCurve=DataInterface() {propertyValues {Curve {keys [<at 0 1>,<at 1 0.5>]}}}, Scale Curve=0.35)
    Index 0 Param 3=50
    Index 0 Param 4=1
    Module.Param0WriteEnabled=true
    Module.Param1WriteEnabled=true
    Module.Param2WriteEnabled=true
    Module.Param3WriteEnabled=true
   )
   &CameraOffset()
   &ScaleColor(
    Scale Alpha=&FloatFromCurve(FloatCurve=DataInterface() {propertyValues {Curve {keys [
         {arriveTangent=9.0725, interpMode=RCIM_Cubic, leaveTangent=9.0725, tangentMode=RCTM_User, time=0, value=0}
         {arriveTangent=1, interpMode=RCIM_Cubic, leaveTangent=1, time=0.1408, value=1}
         {arriveTangent=-1.1638, interpMode=RCIM_Cubic, leaveTangent=-1.1638, time=0.5, value=0.5}
         {interpMode=RCIM_Cubic, leaveTangent=-0.0405, time=1, value=0}
        ]}}})
    Scale RGB=&VectorFromFloat(Value=&FloatFromCurve(FloatCurve=DataInterface() {propertyValues {Curve {keys [<at 0 2>,<at 1 0.2>]}}}))
    ScaleA=true
    ScaleRGB=true
   )
  ]
  rendererClasses [&NiagaraMeshRendererProperties]
  renderers [{rendererClass=&NiagaraMeshRendererProperties, rendererIndex=0}]
 }
 Fountain {
  bEnabled=false
  emitterUpdateScript [
   &EmitterState(Life Cycle Mode=::NewEnumerator1, Loop Duration=2, UseLoopDelay=false)
   &SpawnRate(SpawnRate=90, Use Spawn Probability=false)
  ]
  particleSpawnScript [
   &InitializeParticle(
    Lifetime Max=1.75
    Lifetime Min=1.4
    Lifetime Mode=::NewEnumerator1
    Mass Mode=::NewEnumerator3
    Sprite Rotation Mode=::NewEnumerator2
    Sprite Size Mode=::NewEnumerator1
    Uniform Sprite Size Max=12
    Uniform Sprite Size Min=6
    UsePositionOffset=false
    Write Material Random=true
   )
   &ShapeLocation(
    Box Size=&MakeVector(X=1, Y=$User.Source_Length, Z=1)
    MaskSpawnGroup=false
    Override Seed=false
    Shape Primitive=::NewEnumerator2
    Surface Only=false
    UseOwnerScale=true
   )
   &AddVelocity(
    Cone Angle=32
    Cone Axis=<1 0 1>
    Distribution Along Cone Axis=0.25
    Module.UseConeFalloff=true
    Module.UseVelDistribution=true
    Override Seed=false
    Speed Falloff From Cone Axis=0.5
    Velocity Mode=::NewEnumerator2
    Velocity Speed=&RandomRangeFloat(Minimum=100, Maximum=500)
   )
  ]
  particleUpdateScript [
   &ParticleState()
   &GravityForce()
   &Drag(Drag=0.25, Use Linear Drag=true, Use Rotational Drag=false)
   &ScaleColor(
    Scale Alpha=&FloatFromCurve(FloatCurve=DataInterface() {propertyValues {Curve {keys [<at -0.0046 1.0113 RCIM_Cubic>,<at 1.0011 0.002 RCIM_Cubic>]}}})
    ScaleA=true
    ScaleRGB=true
   )
   &SolveForcesAndVelocity(Clamp Velocity=false, Limit Acceleration=false)
  ]
  rendererClasses [&NiagaraSpriteRendererProperties]
  renderers [{rendererClass=&NiagaraSpriteRendererProperties}]
 }
 Mist {
  simTarget=GPUComputeSim
  emitterUpdateScript [
   &EmitterState()
   &SpawnRate(SpawnRate=8, Use Spawn Probability=false)
  ]
  particleSpawnScript [
   &InitializeParticle(
    Color=<rgba 0.6875 0.808 1 0.1>
    Lifetime Max=4
    Lifetime Min=2.5
    Lifetime Mode=::NewEnumerator1
    Sprite Rotation Mode=::NewEnumerator2
    Sprite Size Mode=::NewEnumerator1
    Uniform Sprite Size Max=180
    Uniform Sprite Size Min=150
    UsePositionOffset=false
    Write Material Random=true
   )
   &BoxLocation(Box Size=$User.BoxSize, MaskSpawnGroup=false, Module.Override Seed=false, Rotate Box=false, Surface Only=false) {enabled=false}
   &ShapeLocation(
    Box Size=&MakeVector(X=1, Y=$User.Source_Length, Z=1)
    MaskSpawnGroup=false
    Offset=<0 0 -20>
    Offset Mode=::NewEnumerator0
    Override Seed=false
    Shape Primitive=::NewEnumerator2
    Surface Only=false
    UseOwnerScale=true
   )
   &AddVelocity(
    Cone Angle=15
    Cone Axis=<0 0 1>
    Module.UseConeFalloff=false
    Module.UseVelDistribution=false
    Override Seed=false
    Velocity Mode=::NewEnumerator2
    Velocity Speed=&RandomRangeFloat(Minimum=150, Maximum=250)
   )
   &ApplyInitialForces()
  ]
  particleUpdateScript [
   &ParticleState()
   &ScaleSpriteSize(
    Uniform Curve Scale=0.5
    Uniform Curve Sprite Scale=DataInterface() {propertyValues {Curve {keys [{tangentMode=RCTM_User, time=0, value=0.5}, {tangentMode=RCTM_User, time=1, value=2}]}}}
   ) {enabled=false, module=ScaleSpriteSize001}
   &DynamicMaterialParameters(
    Index 0 Param 1=2.5
    Index 0 Param 2=4
    Index 0 Param 3=&FloatFromCurve(FloatCurve=DataInterface() {propertyValues {Curve {keys [
         {arriveTangent=-42.914, interpMode=RCIM_Cubic, leaveTangent=-42.914, tangentMode=RCTM_User, time=0, value=30}
         {arriveTangent=-0.0405, interpMode=RCIM_Cubic, leaveTangent=-0.0405, tangentMode=RCTM_User, time=1, value=1}
        ]}}})
    Index 0 Param 4=0.001
    Module.Param0WriteEnabled=true
    Module.Param1WriteEnabled=true
    Module.Param2WriteEnabled=true
    Module.Param3WriteEnabled=true
   ) {module=DynamicMaterialParameters001}
   &Drag(Drag=&RandomRangeFloat(Minimum=0.8, Maximum=1), Use Linear Drag=true, Use Rotational Drag=false)
   &GravityForce(Gravity=&Multiply_Vector(
     A=&RandomRangeVector(Minimum=<0 0 -100>, Maximum=<0 0 -250>)
     B=&VectorFromFloat(Value=1)
    ))
   &ScaleColor(
    Scale Alpha=&FloatFromCurve(
     FloatCurve=DataInterface() {propertyValues {Curve {keys [
         {interpMode=RCIM_Cubic, tangentMode=RCTM_User, time=0, value=0}
         {interpMode=RCIM_Cubic, tangentMode=RCTM_User, time=0.5, value=0.95}
         {tangentMode=RCTM_User, time=1, value=0}
        ]}}}
     Scale Curve=0.5
    )
    Scale RGB=&MakeVectorFromLinearColorRGB(Color=&ColorFromCurve(ColorCurve=DataInterface() {propertyValues {
        AlphaCurve {keys []}
        BlueCurve {keys [<at 0.15 0.7604 RCIM_Cubic>,<at 1 1 RCIM_Cubic>]}
        GreenCurve {keys [<at 0.15 0.5966 RCIM_Cubic>,<at 1 1 RCIM_Cubic>]}
        RedCurve {keys [<at 0.15 0.3599 RCIM_Cubic>,<at 1 1 RCIM_Cubic>]}
       }}))
    ScaleA=true
    ScaleRGB=false
   )
   &SubUVAnimation(Start Frame=55)
   &CameraOffset(Camera Offset Amount=50)
   &SolveForcesAndVelocity(Clamp Velocity=false, Limit Acceleration=false)
   &SpriteRotationRate(Rotation Rate=&RandomRangeFloat(Minimum=-30, Maximum=30))
   &KillParticlesInVolume(Kill Shape=::NewEnumerator2, Origin Offset:Vector3f=<0 0 -100>)
  ]
  rendererClasses [&NiagaraSpriteRendererProperties]
  renderers [{rendererClass=&NiagaraSpriteRendererProperties}]
 }
 MistMesh {
  emitterUpdateScript [
   &EmitterState(Life Cycle Mode=::NewEnumerator1, Loop Behavior=::NewEnumerator1, Loop Duration Mode=::NewEnumerator1, UseLoopDelay=false)
   &SpawnBurst_Instantaneous(
    Spawn Count=&MakeCustomIntFromBool(Bool=$User.bUseBackMist, True INT=1, False INT=0)
    Use Spawn Probability=false
    UseLoopCountLimit=false
   )
  ]
  particleSpawnScript [&InitializeParticle(
    Color=$User.BackMistColor
    Lifetime=$Emitter.CurrentLoopDuration
    Position Offset=<0 0 10>
    Sprite Size=&MakeVector2D(X=278.298, Y=$User.Source_Length)
    Sprite Size Mode=::NewEnumerator4
    UsePositionOffset=true
    Write Material Random=true
   )]
  particleUpdateScript [
   &ParticleState(Kill Particles When Lifetime Has Elapsed=false, Let Infinitely Lived Particles Die When Emitter Deactivates=true)
   &SolveForcesAndVelocity(Clamp Velocity=false, Limit Acceleration=false)
   &SpriteFacingAndAlignment(Facing Coordinate Space=::Local, WriteAlignment=true, WriteFacing=true)
   &DynamicMaterialParameters(
    Index 0 Param 1=$User.Source_Length
    Index 0 Param 2=$User.DropletsBrightness
    Module.Index0Param0Write=true
    Module.Index0Param1Write=true
    Module.Index0Param2Write=true
    Module.Index0Param3Write=true
   )
   &CameraOffset(Camera Offset Amount=10)
  ]
  rendererClasses [&NiagaraSpriteRendererProperties]
  renderers [{rendererClass=&NiagaraSpriteRendererProperties}]
 }
 Splashes {
  bEnabled=false
  simTarget=GPUComputeSim
  emitterUpdateScript [
   &EmitterState()
   &SpawnRate(
    SpawnRate=&Multiply_Float(A=&Divide_Float(A=$User.Source_Length, B=20), B=$User.SplashRateMultiplier)
    Use Spawn Probability=true
   )
  ]
  particleSpawnScript [
   &InitializeParticle(
    Color=$User.SplashColor
    Lifetime=&RandomRangeFloat(Minimum=1, Maximum=2)
    Sprite Rotation Mode=::NewEnumerator2
    Sprite Size Mode=::NewEnumerator1
    Sprite UV Mode=::NewEnumerator3
    Uniform Sprite Size Max=55
    Uniform Sprite Size Min=40
    UsePositionOffset=false
    Write Material Random=true
   )
   &BoxLocation(Box Size=$User.BoxSize, MaskSpawnGroup=false, Module.Override Seed=false, Rotate Box=false, Surface Only=false) {enabled=false}
   &ShapeLocation(
    Box Size=&MakeVector(X=1, Y=$User.Source_Length, Z=1)
    MaskSpawnGroup=false
    Offset=<0 0 -100>
    Offset Mode=::NewEnumerator0
    Override Seed=false
    Shape Primitive=::NewEnumerator2
    Surface Only=false
    UseOwnerScale=true
   )
   &AddVelocity(Velocity Speed Scale=5) {module=AddVelocity001}
   &AddVelocity(
    Cone Angle=14.5765
    Cone Axis=<0.25 0 1>
    Module.UseConeFalloff=false
    Module.UseVelDistribution=false
    Override Seed=false
    Velocity Mode=::NewEnumerator2
    Velocity Speed=&RandomRangeFloat(Minimum=300, Maximum=400)
   )
   &SubUVAnimation(End Frame=3, SubUV Animation Mode:ENiagara_SubUVLookupMode=::NewEnumerator2)
  ]
  particleUpdateScript [
   &ParticleState()
   &GravityForce()
   &Drag(Use Linear Drag=true, Use Rotational Drag=false)
   &SolveForcesAndVelocity(Clamp Velocity=false, Limit Acceleration=false)
   &ScaleColor(
    Scale Alpha=&FloatFromCurve(
     FloatCurve=DataInterface() {propertyValues {Curve {keys [
         {interpMode=RCIM_Cubic, tangentMode=RCTM_User, time=0, value=0}
         {interpMode=RCIM_Cubic, tangentMode=RCTM_User, time=0.4, value=1}
         {tangentMode=RCTM_User, time=1, value=0}
        ]}}}
     Scale Curve=0.35
    )
    ScaleA=true
    ScaleRGB=true
   )
   &ScaleSpriteSize(Uniform Curve Sprite Scale=DataInterface() {propertyValues {Curve {keys [{tangentMode=RCTM_User, time=0, value=0.3}, {tangentMode=RCTM_User, time=1, value=2.5}]}}})
   &SpriteRotationRate(Rotation Rate=&RandomRangeFloat(Minimum=-60, Maximum=60))
   &DynamicMaterialParameters(
    Index 0 Param 1=&FloatFromCurve(FloatCurve=DataInterface() {propertyValues {Curve {keys [<at 0 10>,<at 1 0>]}}})
    Module.Param0WriteEnabled=true
    Module.Param1WriteEnabled=false
    Module.Param2WriteEnabled=false
    Module.Param3WriteEnabled=false
   )
   &KillParticles() {enabled=false}
   &KillParticlesInVolume(Kill Shape=::NewEnumerator2, Origin Offset:Vector3f=<0 0 -120>)
  ]
  rendererClasses [&NiagaraSpriteRendererProperties]
  renderers [{rendererClass=&NiagaraSpriteRendererProperties}]
 }
}
summary {
 systemName=NS_SubtleWaterfallBase_Subtle
 emitters [
  {bEnabled=false, emitterName=Fountain, rendererClasses [&NiagaraSpriteRendererProperties]}
  {bEnabled=false, emitterName=Splashes, simTarget=GPUComputeSim, rendererClasses [&NiagaraSpriteRendererProperties]}
  {emitterName=Mist, simTarget=GPUComputeSim, rendererClasses [&NiagaraSpriteRendererProperties]}
  {emitterName=BaseMesh, rendererClasses [&NiagaraSpriteRendererProperties]}
  {emitterName=FN_Water_Drops_, simTarget=GPUComputeSim, rendererClasses [&NiagaraSpriteRendererProperties]}
  {emitterName=MistMesh, rendererClasses [&NiagaraSpriteRendererProperties]}
  {bEnabled=false, emitterName=FN_Mesh001, rendererClasses [&NiagaraMeshRendererProperties]}
  {bEnabled=false, emitterName=Flat_Ripple, rendererClasses [&NiagaraMeshRendererProperties]}
 ]
 userVariables [
  User.BackMistColor {type=LinearColor, description="", defaultValue {struct=LinearColor, value=<rgba 0.849 0.9331 1 0.1>}}
  User.BaseColor {type=LinearColor, description="", defaultValue {struct=LinearColor, value=<rgba 0.4228 0.7223 0.8958 0.2>}}
  User.BaseWidth {type=NiagaraFloat, defaultValue=200, description=""}
  User.bUseBackMist {type=NiagaraBool, defaultValue=true}
  User.bUseSourceCaustics {type=NiagaraBool, defaultValue=true}
  User.DropletRateMultiplier {type=NiagaraFloat, defaultValue=1, description=""}
  User.DropletsBrightness {type=NiagaraFloat, defaultValue=2, description=""}
  User.RipplesColor {type=LinearColor, description="", defaultValue {struct=LinearColor, value=<rgba 0.0595 0.491 0.5906 0.0075>}}
  User.RipplesSpawnMultiplier {type=NiagaraFloat, defaultValue=1, description=""}
  User.Source_Length {type=NiagaraFloat, defaultValue=2000, description=""}
  User.SplashColor {type=LinearColor, description="", defaultValue {struct=LinearColor, value=<rgba 0.6667 1 1 0.15>}}
  User.SplashRateMultiplier {type=NiagaraFloat, defaultValue=1, description=""}
 ]
}
